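# Example Default Lit material that demonstrates the MF_GenericCracks material function on an Asteria slippery mud surface.
# Mud diffuse, detail normal and mask textures, tiled at twice the base UVs, build a material attributes set, with the mask's blue channel driving both Roughness and Specular.
# Those attributes pass into MF_GenericCracks, whose UVs, Alpha and EmissiveColor inputs are left unconnected.
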Material /Game/Global/Materials/MaterialGym/Materials/GenericCracks/M_GenericCracksExample
graph {
 properties {bUseMaterialAttributes=true, pixelDepthOffsetMode=PDOM_Legacy, usedShadingModels=Default Lit}
 nodes [
  n1 = Multiply(A=TextureCoordinate()) {constB=2}
  n2 = TextureSample(UVs=@n1) {samplerType=SAMPLERTYPE_Masks, texture=/Game/Environments/Asteria/Terrain/Textures/T_Asteria_Mud_Slippery_01_S}
  n3 = SetMaterialAttributes!(
   MaterialAttributes
   Base Color=TextureSample(UVs=@n1) {texture=/Game/Environments/Asteria/Terrain/Textures/T_Asteria_Mud_Slippery_01_D}.RGB
   Normal=TextureSample(UVs=@n1) {samplerType=SAMPLERTYPE_Normal, texture=/Game/Environments/Asteria/Terrain/Textures/T_Asteria_Mud_Slippery_01_Detail_N}.RGB
   Roughness=@n2.B
   Specular=@n2.B
  ) {attributeSetTypes [
    #1
    #2
    #3
    #4
   ]}
  MaterialFunctionCall!(MaterialAttributes=@n3, UVs, Alpha, EmissiveColor) {materialFunction=/Game/Global/Materials/MaterialGym/Materials/GenericCracks/MaterialFunctions/MF_GenericCracks}
 ]
}
